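# Parent material for the EchoFrost set that renders ice only, using material attributes and subsurface shading.
# All surface logic comes from MF_EchoFrost_Ice, called with its ice layer statically enabled over an unmodified set of default material attributes.
# One shared constant stands in for baked curvature, baked AO and painted mask, so the material reads no mesh-specific bakes or painting.
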Material /Hera/Sets/EchoFrost/Materials/M_EchoFrost_Ice_Only_PARENT
graph {
 properties {bAutomaticallySetUsageInEditor=false, bUseMaterialAttributes=true, shadingModel=MSM_Subsurface, usedShadingModels=Subsurface}
 nodes [
  n1 = Constant()
  MaterialFunctionCall!(
   Use EchoFrost Ice Layer=StaticBool() {value=true}
   Input MaterialAttributes=SetMaterialAttributes!(MaterialAttributes)
   "\\ --- Masks --- /"
   EchoFrost Mask
   Tintable Mask
   "\\ --- Dependencies --- /"
   Baked Curvature=@n1
   Baked AO=@n1
   Painted Mask=@n1
   DFAO
   Z Gradient
   WS Noise Masks
  ) {materialFunction=/Hera/Sets/EchoFrost/Materials/MaterialFunctions/MF_EchoFrost_Ice}
 ]
}
